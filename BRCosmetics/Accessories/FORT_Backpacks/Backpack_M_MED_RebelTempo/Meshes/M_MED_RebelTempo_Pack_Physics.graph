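# Physics asset for the RebelTempo backpack cosmetic, with bodies for the pack, a sword base, and dynamic sword and strap bones.
# Each body is a single capsule, largest on the pack and smallest on base_sword.
# The sword and strap each constrain to base_sword with tight swing and twist limits, while base_sword links to the pack with free angular motion.
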PhysicsAsset /BRCosmetics/Accessories/FORT_Backpacks/Backpack_M_MED_RebelTempo/Meshes/M_MED_RebelTempo_Pack_Physics
bodies {
 base_sword [{shapeType=Capsule, length=0.505, radius=0.505, shapeName=base_sword_capsule, rotation=<rot 0 -90 0>}]
 dyn_strap [{shapeType=Capsule, length=2, radius=1.5, shapeName=dyn_sword_capsule, center=<2 0 0>, rotation=<rot 90 0 0>}]
 dyn_sword [{shapeType=Capsule, length=2, radius=1.5, shapeName=dyn_sword_capsule, center=<2 0 0>, rotation=<rot 90 0 0>}]
 pack [{shapeType=Capsule, length=9.334, radius=4.8827, shapeName=pack_capsule, center=<0 -1.382 0.0404>, rotation=<rot 0 -90 -90>}]
}
constraints [
 {
  bone1Name=base_sword
  bone2Name=pack
  swing1LimitDegrees=45
  swing1Motion=Free
  swing2LimitDegrees=45
  swing2Motion=Free
  twistLimitDegrees=45
  twistMotion=Free
 }
 {bone1Name=dyn_sword, bone2Name=base_sword, swing1LimitDegrees=5, swing2LimitDegrees=10, twistLimitDegrees=5}
 {bone1Name=dyn_strap, bone2Name=base_sword, swing1LimitDegrees=5, swing2LimitDegrees=10, twistLimitDegrees=5}
]
